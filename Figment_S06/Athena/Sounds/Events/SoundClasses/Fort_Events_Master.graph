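# Fort_Events_Master is a SoundClass that groups event sounds for the Figment_S06 Athena mode under the Fort_Athena_SFX parent class.
# Its children are four shared Fort_Events priority classes, from ExtHigh to Low, plus a local Fort_Events_Mapwide class.
# The priority classes come from the base Game sound folders, so this node links them into the Figment Athena SFX hierarchy.
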SoundClass /Figment_S06/Athena/Sounds/Events/SoundClasses/Fort_Events_Master
parent=/Figment_S06/Sounds/FortSoundClassesAndModes/Fort_Athena_SFX
children [
 /Game/Sounds/FortSoundClassesAndModes/Fort_Events_Priority_ExtHigh
 /Game/Sounds/FortSoundClassesAndModes/Fort_Events_Priority_High
 /Game/Sounds/FortSoundClassesAndModes/Fort_Events_Priority_Medium
 /Game/Sounds/FortSoundClassesAndModes/Fort_Events_Priority_Low
 &/Fort_Events_Mapwide
]
